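# Unlit translucent particle-sprite material for glowing smoke in Festivus firework effects.
# Two spark colors are blended, HDR-tinted by the ColorIntensity dynamic parameter, then hue-shifted by particle red, scaled by a clamped value that a static switch can take from particle blue.
# Particle green and the smoke texture's red channel scale that color before emissive adjustments.
# Particle alpha times texture green forms the smoke mask.
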Material /Figment_S10/Athena/Effects/Materials/Festivus/M_Fesitvus_Firework_Smoke_Emissive
graph {
 properties {
  bUsedWithParticleSprites=true
  blendMode=BLEND_Translucent
  pixelDepthOffsetMode=PDOM_Legacy
  shadingModel=MSM_Unlit
  translucencyLightingMode=TLM_VolumetricPerVertexDirectional
  usedShadingModels=Unlit
 }
 nodes [
  n1 = ParticleColor()
  n2 = ParticleColor()
  n3 = StaticSwitchParameter(True=@n2.B, False=Constant() {r=1}) {parameterName=USe Blue Color for HueShift}
  n4 = ScalarParameter() {defaultValue=0.25, parameterName=Min}
  n5 = ScalarParameter() {defaultValue=0.75, parameterName=Max}
  n6 = LinearInterpolate(
   A=VectorParameter() {parameterName=SparkColor0, defaultValue=<rgba 1 0.0978 0 0>}.RGB
   B=VectorParameter() {parameterName=SparkColor1, defaultValue=<rgba 1 0.2289 0 0>}.RGB
   Alpha=ScalarParameter() {parameterName=ColorBlend}
  )
  n7 = DynamicParameter() {defaultValue=<rgba 1 1 -1 1>, paramNames [ColorIntensity, PanSmoke, OffsetMesh, LineWidth]}
  n8 = MaterialFunctionCall(
   Hue Shift Percentage=Multiply(A=@n2.R, B=Clamp(None=@n3, Min=@n4, Max=@n5) {maxDefault=0.75, minDefault=0.25})
   Texture=MaterialFunctionCall(LDR Input=@n6, HDR Tint=Multiply(A=@n7.ColorIntensity, B=@n6)) {materialFunction=&DeriveHDRfromLDR}.RemappedRange
  ) {materialFunction=&HueShift}
  n9 = TextureSample() {samplerType=SAMPLERTYPE_LinearColor, texture=/Game/Effects/Fort_Effects/Textures/Smoke/T_Smoke_8x8_01}
  MaterialFunctionCall(Emissive=Multiply(A=Multiply(A=@n1.G, B=@n8.Result), B=@n9.R)) {materialFunction=&MF_FX_EmissiveAdjustments}
  Multiply(A=ParticleColor().A, B=@n9.G)
  ParticleSubUV!(UVs, Tex, Apply View MipBias) {texture=/Game/Effects/Fort_Effects/BluePrints/BeeHive/T_SmokeFlip_03_8x8}
 ]
}
